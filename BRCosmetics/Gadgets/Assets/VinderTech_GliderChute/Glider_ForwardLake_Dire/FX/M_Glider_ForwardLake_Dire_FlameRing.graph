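# Unlit, translucent, two-sided material for a flame ring effect on the Forward Lake Dire glider, built for Niagara mesh particles.
# Particle color passes through an emissive adjustment function to supply the glow.
# Opacity multiplies particle alpha, a cloud texture's green channel, a Fresnel function with adjustable exponent, and depth fade for soft intersections.
# Zero U tiling makes the cloud sample vary only along V, producing streaks or bands.
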Material /BRCosmetics/Gadgets/Assets/VinderTech_GliderChute/Glider_ForwardLake_Dire/FX/M_Glider_ForwardLake_Dire_FlameRing
graph {
 properties {
  bUsedWithNiagaraMeshParticles=true
  blendMode=BLEND_Translucent
  pixelDepthOffsetMode=PDOM_Legacy
  shadingModel=MSM_Unlit
  twoSided=true
  usedShadingModels=Unlit
 }
 nodes [
  n1 = ParticleColor()
  MaterialFunctionCall(Emissive=@n1.RGB) {materialFunction=&MF_FX_EmissiveAdjustments}
  n2 = TextureSample(UVs=TextureCoordinate() {uTiling=0}) {texture=/Game/Missions/Primary/DeliverTheBomb/Textures/T_DtB_Cloud}
  n3 = ScalarParameter() {defaultValue=2, parameterName=Exponent}
  Multiply(
   A=Multiply(
    A=Multiply(A=@n1.A, B=@n2.G)
    B=MaterialFunctionCall(Exponent=@n3) {materialFunction=&MF_Fresnel_FX}.Fresnel
   )
   B=DepthFade(Opacity=Constant() {r=1}, FadeDistance=ScalarParameter() {defaultValue=16, parameterName=DepthFade_Distance})
  )
 ]
}
