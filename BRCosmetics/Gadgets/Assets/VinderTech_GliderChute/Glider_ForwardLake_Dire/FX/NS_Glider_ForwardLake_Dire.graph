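# Niagara effect for the ForwardLake Dire glider cosmetic, rendering twin flame exhausts with smoke, glow lighting and wind lines.
# FlameRing meshes, Lighting sprites and lights, and Smoke meshes attach to two skeletal mesh bones or sockets.
# System effect intensity and fade-in drive brightness and smoke, while timeline curves add pulsed bursts scaled by User.ForwardDot.
# WindLines spawn on a torus steered by User.RightDot, with front-end modifiers adjusting several values.
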NiagaraSystem /BRCosmetics/Gadgets/Assets/VinderTech_GliderChute/Glider_ForwardLake_Dire/FX/NS_Glider_ForwardLake_Dire
dependencies {
 usedDataInterfaces [
  &NiagaraDataInterfaceCurve
  &NiagaraDataInterfaceArrayInt32
  &NiagaraDataInterfaceArrayFloat3
  &NiagaraDataInterfaceVectorCurve
  &NiagaraDataInterfaceColorCurve
  &NiagaraDataInterfaceSkeletalMesh
  &NiagaraDataInterfaceArrayFloat
 ]
 usedDynamicInputs [
  &SelectVectorFromArray
  &SelectFloatFromArray
  &MakeIntFromFloat
 ]
 usedModules [
  &SystemState
  &self:SystemSpawnScript.SystemScriptSource.SystemScriptGraph.NiagaraNodeAssignment_3.SetVariables_#1
  &Timeline
  &FN_Integrator
  &InterpolateOverTime
  &self:SystemSpawnScript.SystemScriptSource.SystemScriptGraph.NiagaraNodeAssignment_1.SetVariables_#2
  &self:SystemSpawnScript.SystemScriptSource.SystemScriptGraph.NiagaraNodeAssignment_2.SetVariables_#3
 ]
}
emitters {
 FlameRing {
  emitterUpdateScript [
   &EmitterState(
    Inactive Response=::NewEnumerator1
    Life Cycle Mode=::NewEnumerator1
    Loop Behavior=::NewEnumerator1
    Loop Duration Mode=::NewEnumerator1
    UseLoopDelay=false
   )
   &SpawnRate(SpawnRate=30, Use Spawn Probability=false)
  ]
  particleSpawnScript [
   &InitializeParticle(
    Lifetime Max=0.3
    Lifetime Min=0.2
    Lifetime Mode=::NewEnumerator1
    Mesh Scale=<0.12 0.04 0.04>
    Mesh Scale Mode=::NewEnumerator4
    UsePositionOffset=true
    Write Material Random=true
   )
   &FN_SetSpawnGroup(Number of Spawn Groups=2)
   &InitialMeshOrientation(
    Mesh Axis To Orient=<-1 0 0>
    Mesh Orientation Mode=::NewEnumerator2
    Orientation Vector=<0 1 0>
    Rotation=&RandomRangeVector(Minimum=<0 0 0>, Maximum=<1 0 0>)
    Use Rotation Vector=true
   )
  ]
  particleUpdateScript [
   &ParticleState()
   &SkeletalMeshLocation(
    Bone / Socket Index=&MergeHelper_IntSwitch(0=0, 1=1, 2=?, 3=?, 4=?, 5=?, Index=&SpawnGroup(), Num=::NewEnumerator0)
    MaskSpawnGroup=false
    Mesh Sampling Type=::NewEnumerator5
    Sampled Position Offset=<0 -10 0>
    Skeletal Mesh=$System.Skeletal Mesh
    Socket Sampling Mode=::NewEnumerator5
   )
   &ScaleMeshSize(Scale Factor=&MakeVector(
     X=&Multiply_Float(
      A=&FloatFromCurve(FloatCurve=DataInterface() {propertyValues {Curve {keys [<at 0 0.3>,<at 1 1>]}}})
      B=2
     )
     Y=1.5
     Z=1.5
    ))
   &ScaleColor(
    Scale Alpha=&FloatFromCurve(
     FloatCurve=DataInterface() {propertyValues {Curve {keys [
         {arriveTangent=9.0725, interpMode=RCIM_Cubic, leaveTangent=9.0725, tangentMode=RCTM_User, time=0, value=0}
         {arriveTangent=-1.6443, interpMode=RCIM_Cubic, leaveTangent=-1.6443, tangentMode=RCTM_User, time=0.1408, value=0.9847}
         {arriveTangent=-0.0405, interpMode=RCIM_Cubic, leaveTangent=-0.0405, tangentMode=RCTM_User, time=1, value=0}
        ]}}}
     Scale Curve=&Add_Float(
      A=&Lerp_Float(A=1, B=13, Alpha=$System.EffectIntensity)
      B=&FloatFromCurve(
       FloatCurve=DataInterface() {propertyValues {Curve {keys [<at 0 0 RCIM_Constant>,<at 0.005 3>,<at 0.07 0 RCIM_Constant>,<at 0.2 3>,<at 0.4 0 RCIM_Constant>,<at 0.6 3>,<at 1 0>]}}}
       CurveIndex=$System.Timeline003.NormalizedTime
       Scale Curve=&FN_FrontEndFloatModifier(Gameplay=$User.ForwardDot, Front End=1)
      )
     )
    )
    Scale RGB=&Multiply_VectorByFloat(
     Vector=&MakeVectorFromLinearColorRGB(Color=&ColorFromCurve(ColorCurve=DataInterface() {propertyValues {
         AlphaCurve {keys [<at 0 1 RCIM_Cubic>]}
         BlueCurve {keys [<at 0.0203 0.15>, <at 0.0654 0.05>, {arriveTangent=-15.6046, time=0.1787, value=0.05}, <at 0.6131 0.05>, <at 0.8857 0.05>]}
         GreenCurve {keys [<at 0.0203 0.6>, <at 0.0654 0.2083>, {arriveTangent=-5.6465, time=0.1787, value=0.1767}, <at 0.6131 0.1133>, <at 0.8857 0.05>]}
         RedCurve {keys [<at 0.0203 1.5>, <at 0.0654 1>, {arriveTangent=-0.716, time=0.1787, value=1}, <at 0.6131 1>, <at 0.8857 1>]}
        }}))
     Float=&Add_Float(
      A=1
      B=&FloatFromCurve(
       FloatCurve=DataInterface() {propertyValues {Curve {keys [<at 0 0 RCIM_Constant>,<at 0.005 0.5>,<at 0.07 0 RCIM_Constant>,<at 0.2 0.5>,<at 0.4 0 RCIM_Constant>,<at 0.6 0.5>,<at 1 0>]}}}
       CurveIndex=$System.Timeline003.NormalizedTime
       Scale Curve=&FN_FrontEndFloatModifier(Gameplay=$User.ForwardDot, Front End=1)
      )
     )
    )
    ScaleA=true
    ScaleRGB=true
   )
   &DynamicMaterialParameters(
    Index 0 Param 1=&RandomRangeFloat(Minimum=0, Maximum=3)
    Module.Index0Float0Write=true
    Module.Index0Float1Write=true
    Module.Index0Float2Write=true
    Module.Index0Float3Write=true
   )
  ]
  rendererClasses [&NiagaraMeshRendererProperties]
  renderers [{rendererClass=&NiagaraMeshRendererProperties, rendererIndex=0}]
 }
 Lighting {
  emitterUpdateScript [
   &EmitterState(
    Inactive Response=::NewEnumerator1
    Life Cycle Mode=::NewEnumerator1
    Loop Behavior=::NewEnumerator1
    Loop Duration Mode=::NewEnumerator1
    UseLoopDelay=false
   )
   &SpawnBurst_Instantaneous(Spawn Count=2, Use Spawn Probability=false, UseLoopCountLimit=true)
  ]
  particleSpawnScript [
   &FN_SetSpawnGroup(Number of Spawn Groups=2)
   &InitializeParticle(
    Color=<rgba 25 2.0833 0 1>
    Lifetime=3
    Sprite Size Mode=::NewEnumerator3
    Uniform Sprite Size=3.75
    UsePositionOffset=true
    Write Material Random=true
   )
   &DynamicMaterialParameters(
    Index 0 Param 1=1
    Index 0 Param 2=1
    Index 0 Param 3=1
    Index 0 Param 4=1
    Module.Param0WriteEnabled=true
    Module.Param1WriteEnabled=true
    Module.Param2WriteEnabled=true
    Module.Param3WriteEnabled=true
   )
  ]
  particleUpdateScript [
   &ParticleState(
    Kill Particles When Lifetime Has Elapsed=false
    Let Infinitely Lived Particles Die When Emitter Deactivates=true
    Loop Particles Lifetime=true
   )
   &SkeletalMeshLocation(
    Bone / Socket Index=&SelectIntFromArray(Direct Array Index=&SpawnGroup(), Int Selection Array=DataInterface() {propertyValues {IntData [0,1]}})
    MaskSpawnGroup=false
    Mesh Sampling Type=::NewEnumerator5
    Skeletal Mesh=$System.Skeletal Mesh
    Socket Sampling Mode=::NewEnumerator5
   )
   &CameraOffset()
   &ScaleColor(
    Scale Alpha=&FloatFromCurve(
     FloatCurve=DataInterface() {propertyValues {Curve {keys [
         <at 0 1 RCIM_Cubic>
         {arriveTangent=0.1177, interpMode=RCIM_Cubic, leaveTangent=0.1177, time=0.0746, value=0.8987}
         {arriveTangent=-0.0227, interpMode=RCIM_Cubic, leaveTangent=-0.0227, tangentMode=RCTM_User, time=0.1682, value=1.0198}
         {arriveTangent=-0.7897, interpMode=RCIM_Cubic, leaveTangent=-0.7897, time=0.2019, value=0.7756}
         {arriveTangent=0.8345, interpMode=RCIM_Cubic, leaveTangent=0.8345, time=0.2748, value=0.9356}
         {arriveTangent=0.5127, interpMode=RCIM_Cubic, leaveTangent=0.5127, time=0.3464, value=0.8961}
         {arriveTangent=-0.6889, interpMode=RCIM_Cubic, leaveTangent=-0.6889, time=0.4074, value=1.0036}
         {arriveTangent=-0.8737, interpMode=RCIM_Cubic, leaveTangent=-0.8737, time=0.4365, value=0.834}
         {arriveTangent=-0.4832, interpMode=RCIM_Cubic, leaveTangent=-0.4832, time=0.4871, value=0.9339}
         {arriveTangent=0.1715, interpMode=RCIM_Cubic, leaveTangent=0.1715, time=0.5331, value=0.7873}
         {arriveTangent=0.795, interpMode=RCIM_Cubic, leaveTangent=0.795, time=0.6112, value=0.9552}
         {arriveTangent=0.9015, interpMode=RCIM_Cubic, leaveTangent=0.9015, time=0.69, value=0.912}
         {arriveTangent=-0.4254, interpMode=RCIM_Cubic, leaveTangent=-0.4254, time=0.7536, value=1.0835}
         {arriveTangent=-1.4701, interpMode=RCIM_Cubic, leaveTangent=-1.4701, time=0.7995, value=0.8654}
         {arriveTangent=-0.4912, interpMode=RCIM_Cubic, leaveTangent=-0.4912, time=0.8674, value=0.9161}
         {arriveTangent=0.6326, interpMode=RCIM_Cubic, leaveTangent=0.6326, time=0.9138, value=0.8093}
         {interpMode=RCIM_Cubic, leaveTangent=0.1355, time=1, value=1}
        ]}}}
     Scale Curve=&FloatFromCurve(
      FloatCurve=DataInterface() {propertyValues {Curve {defaultValue=1, keys [{tangentMode=RCTM_User, time=0, value=0}, {tangentMode=RCTM_User, time=1, value=1}]}}}
      CurveIndex=$System.EffectIntensity
      Scale Curve=0.3
     )
    )
    Scale RGB=&VectorFromFloat(Value=&FloatFromCurve(
      FloatCurve=DataInterface() {propertyValues {Curve {keys [
          <at 0 1 RCIM_Cubic>
          {arriveTangent=0.1177, interpMode=RCIM_Cubic, leaveTangent=0.1177, time=0.0685, value=0.5809}
          {arriveTangent=-0.0227, interpMode=RCIM_Cubic, leaveTangent=-0.0227, tangentMode=RCTM_User, time=0.1682, value=1.0198}
          {arriveTangent=-0.7897, interpMode=RCIM_Cubic, leaveTangent=-0.7897, time=0.1958, value=0.4578}
          {arriveTangent=0.8345, interpMode=RCIM_Cubic, leaveTangent=0.8345, time=0.2748, value=0.9356}
          {arriveTangent=0.5127, interpMode=RCIM_Cubic, leaveTangent=0.5127, time=0.3403, value=0.5784}
          {arriveTangent=-0.6889, interpMode=RCIM_Cubic, leaveTangent=-0.6889, time=0.4074, value=1.0036}
          {arriveTangent=-3.5031, interpMode=RCIM_Cubic, leaveTangent=-3.5031, time=0.4304, value=0.5163}
          {arriveTangent=-0.4832, interpMode=RCIM_Cubic, leaveTangent=-0.4832, time=0.4871, value=0.7243}
          {arriveTangent=1.8607, interpMode=RCIM_Cubic, leaveTangent=1.8607, time=0.527, value=0.4696}
          {arriveTangent=0.795, interpMode=RCIM_Cubic, leaveTangent=0.795, time=0.6112, value=0.9552}
          {arriveTangent=0.9015, interpMode=RCIM_Cubic, leaveTangent=0.9015, time=0.6839, value=0.5943}
          {arriveTangent=-0.4254, interpMode=RCIM_Cubic, leaveTangent=-0.4254, time=0.7536, value=1.0835}
          {arriveTangent=-3.3116, interpMode=RCIM_Cubic, leaveTangent=-3.3116, time=0.7934, value=0.5477}
          {arriveTangent=-0.4912, interpMode=RCIM_Cubic, leaveTangent=-0.4912, time=0.8674, value=0.7065}
          {arriveTangent=2.2144, interpMode=RCIM_Cubic, leaveTangent=2.2144, time=0.9076, value=0.4916}
          {interpMode=RCIM_Cubic, leaveTangent=0.1355, time=1, value=1}
         ]}}}
      Scale Curve=&FloatFromCurve(
       FloatCurve=DataInterface() {propertyValues {Curve {keys [
           {arriveTangent=0.0801, interpMode=RCIM_Cubic, leaveTangent=0.0801, tangentMode=RCTM_User, time=0, value=1}
           {arriveTangent=0.097, interpMode=RCIM_Cubic, leaveTangent=0.097, tangentMode=RCTM_User, time=0.7, value=1}
           {arriveTangent=4.5194, interpMode=RCIM_Cubic, leaveTangent=4.5194, tangentMode=RCTM_User, time=1, value=1.5}
          ]}}}
       CurveIndex=$System.EffectIntensity
       Scale Curve=10
      )
     ))
    ScaleA=true
    ScaleRGB=true
   )
   &FN_Light_Attributes(Create Exponent Attribute=true, Create Radius Attribute=true, Exponent=1, Radius=20) {module=FN_Light_Attributes001}
   /SetVariables(
    Particles.LightColor=&Multiply_LinearColorByFloat(
     Linear Color=&LerpLinearColors(
      StartColor=&MultiplyLinearColor(
       Linear Color=<rgba 1.5 0.075 0 1>
       Scale Factor=&MakeFloat4FromVectorAndFloat(
        "Vector (RGB)"=&VectorFromFloat(Value=0.2)
        "Float (Alpha)"=&Multiply_Float(A=1, B=0.01)
       )
      )
      EndColor=&MultiplyLinearColor(
       Linear Color=<rgba 1.5 0.114 0.015 1>
       Scale Factor=&MakeFloat4FromVectorAndFloat(
        "Vector (RGB)"=&VectorFromFloat(Value=1.2)
        "Float (Alpha)"=&Multiply_Float(A=1, B=0.11)
       )
      )
      LerpFactor=$System.EffectIntensity
     )
     Float=&Multiply_Float(
      A=&Add_Float(
       A=1
       B=&FloatFromCurve(
        FloatCurve=DataInterface() {propertyValues {Curve {keys [<at 0 0 RCIM_Constant>,<at 0.005 0.5>,<at 0.07 0 RCIM_Constant>,<at 0.2 0.5>,<at 0.4 0 RCIM_Constant>,<at 0.6 0.5>,<at 1 0>]}}}
        CurveIndex=$System.Timeline003.NormalizedTime
        Scale Curve=&FN_FrontEndFloatModifier(Gameplay=$User.ForwardDot, Front End=1)
       )
      )
      B=0.5
     )
    )
    Particles.LightPosition:NiagaraPosition=&AddVectorToPosition(Position=$Particles.Position, Vector=<0 -2 0>)
   ) {isSetParameters=true}
   /SetVariables(
    Particles.FakeLightCColor:LinearColor=&Multiply_LinearColorByFloat(
     Linear Color=&LerpLinearColors(
      StartColor=&MultiplyLinearColor(
       Linear Color=<rgba 1.5 0.075 0 1>
       Scale Factor=&MakeFloat4FromVectorAndFloat(
        "Vector (RGB)"=&VectorFromFloat(Value=0.5)
        "Float (Alpha)"=&Multiply_Float(A=1, B=0.01)
       )
      )
      EndColor=&MultiplyLinearColor(
       Linear Color=<rgba 1.5 0.114 0.015 1>
       Scale Factor=&MakeFloat4FromVectorAndFloat(
        "Vector (RGB)"=&VectorFromFloat(Value=1.2)
        "Float (Alpha)"=&Multiply_Float(A=1, B=0.07)
       )
      )
      LerpFactor=$System.EffectIntensity
     )
     Float=&Add_Float(
      A=1
      B=&FloatFromCurve(
       FloatCurve=DataInterface() {propertyValues {Curve {keys [<at 0 0 RCIM_Constant>,<at 0.005 0.5>,<at 0.07 0 RCIM_Constant>,<at 0.2 0.5>,<at 0.4 0 RCIM_Constant>,<at 0.6 0.5>,<at 1 0>]}}}
       CurveIndex=$System.Timeline003.NormalizedTime
       Scale Curve=&FN_FrontEndFloatModifier(Gameplay=$User.ForwardDot, Front End=1)
      )
     )
    )
    Particles.FakeLightPosition:NiagaraPosition=&AddVectorToPosition(Position=$Particles.Position, Vector=<0 -2 0>)
   ) {isSetParameters=true}
   /SetVariables(
    Particles.SpriteGlowCameraOffset=20
    Particles.SpriteGlowColor:LinearColor=&Multiply_LinearColorByFloat(
     Linear Color=&LerpLinearColors(
      StartColor=&MultiplyLinearColor(
       Linear Color=<rgba 1.5 0.1525 0.03 1>
       Scale Factor=&MakeFloat4FromVectorAndFloat(
        "Vector (RGB)"=&VectorFromFloat(Value=1)
        "Float (Alpha)"=&Multiply_Float(A=1, B=0.01)
       )
      )
      EndColor=&MultiplyLinearColor(
       Linear Color=<rgba 1.5 0.3125 0.075 1>
       Scale Factor=&MakeFloat4FromVectorAndFloat(
        "Vector (RGB)"=&VectorFromFloat(Value=10)
        "Float (Alpha)"=&Multiply_Float(A=1, B=0.15)
       )
      )
      LerpFactor=$System.EffectIntensity
     )
     Float=&Add_Float(
      A=1
      B=&FloatFromCurve(
       FloatCurve=DataInterface() {propertyValues {Curve {keys [<at 0 0 RCIM_Constant>,<at 0.005 0.5>,<at 0.07 0 RCIM_Constant>,<at 0.2 0.5>,<at 0.4 0 RCIM_Constant>,<at 0.6 0.5>,<at 1 0>]}}}
       CurveIndex=$System.Timeline003.NormalizedTime
       Scale Curve=&FN_FrontEndFloatModifier(Gameplay=$User.ForwardDot, Front End=1)
      )
     )
    )
    Particles.SpriteGlowPosition:NiagaraPosition=&AddVectorToPosition(Position=$Particles.Position, Vector=<0 8 0>)
    Particles.SpriteGlowSize:Vector2f=<90 90>
   ) {isSetParameters=true}
  ]
  rendererClasses [&NiagaraSpriteRendererProperties, &NiagaraLightRendererProperties]
  renderers [
   {rendererClass=&NiagaraSpriteRendererProperties}
   {rendererClass=&NiagaraLightRendererProperties, rendererIndex=1}
   {rendererClass=&NiagaraSpriteRendererProperties, rendererIndex=2}
  ]
 }
 Smoke {
  emitterUpdateScript [
   &EmitterState()
   &SpawnRate(
    Spawn Probability=$System.FadeIn
    SpawnRate=&Lerp_Float(
     A=0
     B=30
     Alpha=&FloatFromCurve(FloatCurve=DataInterface() {propertyValues {Curve {keys [<at 0.8 0>,<at 1 1>]}}}, CurveIndex=$System.EffectIntensity)
    )
    Use Spawn Probability=true
   )
   &SpawnRate(
    Spawn Probability=&FloatFromCurve(
     FloatCurve=DataInterface() {propertyValues {Curve {keys [<at 0 0 RCIM_Constant>,<at 0.005 2>,<at 0.07 0 RCIM_Constant>,<at 0.2 1.5>,<at 0.4 0 RCIM_Constant>,<at 0.6 3>,<at 1 0>]}}}
     CurveIndex=$System.Timeline003.NormalizedTime
     Scale Curve=&FN_FrontEndFloatModifier(Gameplay=$User.ForwardDot, Front End=1)
    )
    SpawnRate=40
    Use Spawn Probability=true
   ) {module=SpawnRate001}
  ]
  particleSpawnScript [
   &PartitionParticles()
   &InitializeParticle(
    Color=&MultiplyLinearColor(
     Linear Color=<rgba 1 1 1 1>
     Scale Factor=&MakeVector4_FromFloat(Float=&FN_FrontEndFloatModifier(Gameplay=1, Front End=1))
    )
    Lifetime Max=0.7
    Lifetime Min=0.5
    Lifetime Mode=::NewEnumerator1
    Mesh Scale=&Multiply_VectorByFloat(Vector=&RandomRangeVector(Minimum=<3.5 2 2.5>, Maximum=<4.5 3.5 3.5>), Float=0.4)
    Mesh Scale Mode=::NewEnumerator4
    UsePositionOffset=false
    Write Material Random=true
   )
   &AddVelocity(
    Scale Added Velocity=&Multiply_VectorByFloat(
     Vector=&Multiply_VectorByFloat(Vector=<2.5 2.5 2.5>, Float=1)
     Float=&Lerp_Float(A=1, B=0.6, Alpha=$System.InterpolateOverTime.InterpolateOverTime_Float.Moving Average)
    )
    Velocity=&Lerp_Vector(A=<0 100 0>, B=<0 300 0>, Alpha=$System.EffectIntensity)
   )
   &SkeletalMeshLocation(
    Bone / Socket Index=&SelectIntFromArray(Direct Array Index=$Particles.PartitionParticles.Partition, Int Selection Array=DataInterface() {propertyValues {IntData [0,1]}})
    MaskSpawnGroup=false
    Mesh Sampling Type=::NewEnumerator5
    Orientation Sampling=::NewEnumerator1
    Skeletal Mesh=$System.Skeletal Mesh
    Socket Sampling Mode=::NewEnumerator5
   )
   &ShapeLocation(MaskSpawnGroup=false, Override Seed=false, Sphere Radius=1, UseOwnerScale=true)
   &AddVelocityInCone(Cone Angle=50, Cone Axis=<0 1 0>, Module.Use Velocity Falloff On Cone Axis=false, Velocity Strength=66.6) {module=AddVelocityInCone001}
   &InitialMeshOrientation(
    Mesh Orientation Mode=::NewEnumerator2
    Rotation=&RandomRangeVector(Minimum=<0 0 0>, Maximum=<1 0 0>)
    Use Rotation Vector=true
   )
   &CameraOffset(Camera Offset Amount=-10)
  ]
  particleUpdateScript [
   &Drag(Use Linear Drag=true, Use Rotational Drag=true)
   &SolveForcesAndVelocity(Clamp Velocity=false, Limit Acceleration=false)
   &ParticleState()
   &ScaleMeshSize(Scale Factor=&Multiply_VectorByFloat(
     Vector=&Multiply_VectorByFloat(
      Vector=&Multiply_Vector(
       A=<1 1 1>
       B=&VectorFromCurve(VectorCurve=DataInterface() {propertyValues {XCurve {keys [<at 0 5>,<at 0.2 1>,<at 1 1>]}, YCurve {keys [<at 0 1>,<at 1 1>]}, ZCurve {keys [<at 0 1>,<at 1 1>]}}})
      )
      Float=&RandomRangeFloat(Minimum=1, Maximum=1.5)
     )
     Float=&FloatFromCurve(
      FloatCurve=DataInterface() {propertyValues {Curve {keys [
          {interpMode=RCIM_Cubic, tangentMode=RCTM_User, time=0, value=0.2}
          {arriveTangent=1.3433, tangentMode=RCTM_User, time=0.2, value=0.5}
          {arriveTangent=1.4526, interpMode=RCIM_Cubic, leaveTangent=1.4526, tangentMode=RCTM_User, time=1, value=1.5}
         ]}}}
      Scale Curve=&Multiply_Float(A=0.83, B=1)
     )
    ))
   &DynamicMaterialParameters(
    Index 0 Param 1=&FloatFromCurve(
     FloatCurve=DataInterface() {propertyValues {Curve {keys [
         {interpMode=RCIM_Cubic, tangentMode=RCTM_User, time=0, value=0}
         <at 0.2 0.55>
         {arriveTangent=-1.0223, interpMode=RCIM_Cubic, leaveTangent=-1.0223, tangentMode=RCTM_User, time=1, value=1}
        ]}}}
     Scale Curve=&RandomRangeFloat(Minimum=0.7, Maximum=1.5)
    )
    Index 0 Param 4=&Lerp_Float(
     A=0.6
     B=0
     Alpha=&FloatFromCurve(FloatCurve=DataInterface() {propertyValues {Curve {keys [
          {arriveTangent=0.7438, interpMode=RCIM_Cubic, leaveTangent=0.7438, tangentMode=RCTM_User, time=0, value=0}
          {arriveTangent=0.0023, interpMode=RCIM_Cubic, leaveTangent=0.0023, tangentMode=RCTM_User, time=1, value=1}
         ]}}})
    )
    Module.Index0Param0Write=true
    Module.Index0Param1Write=true
    Module.Index0Param2Write=true
    Module.Index0Param3Write=true
   ) {module=DynamicMaterialParameters001}
   &ScaleColor(
    Scale Alpha=&FloatFromCurve(
     FloatCurve=DataInterface() {propertyValues {Curve {keys [
         {arriveTangent=2.8927, interpMode=RCIM_Cubic, leaveTangent=2.8927, tangentMode=RCTM_User, time=0, value=0}
         <at 0.05 1>
         {interpMode=RCIM_Cubic, tangentMode=RCTM_Break, time=0.5, value=1}
         <at 1 0>
        ]}}}
     Scale Curve=&FloatFromCurve(
      FloatCurve=DataInterface() {propertyValues {Curve {keys [
          {arriveTangent=2.8927, interpMode=RCIM_Cubic, leaveTangent=2.8927, tangentMode=RCTM_User, time=0, value=0.3}
          <at 0.01 1>
          {interpMode=RCIM_Cubic, tangentMode=RCTM_Break, time=0.6, value=1}
          <at 1 0.3>
         ]}}}
      CurveIndex=$System.Timeline003.NormalizedTime
      Scale Curve=&Multiply_Float(A=0.5, B=$System.FadeIn)
     )
    )
    Scale RGB=&MakeVectorFromLinearColorRGB(Color=&ColorFromCurve(ColorCurve=DataInterface() {propertyValues {
        AlphaCurve {keys []}
        BlueCurve {keys [<at 0.0108 0.8 RCIM_Cubic>,<at 0.1099 0.4 RCIM_Cubic>]}
        GreenCurve {keys [<at 0.0108 8.64 RCIM_Cubic>,<at 0.1099 0.9333 RCIM_Cubic>]}
        RedCurve {keys [<at 0.0108 40 RCIM_Cubic>,<at 0.1099 2 RCIM_Cubic>]}
       }}))
    ScaleA=true
    ScaleRGB=true
   )
  ]
  rendererClasses [&NiagaraMeshRendererProperties]
  renderers [{rendererClass=&NiagaraMeshRendererProperties, rendererIndex=0}]
 }
 WindLines001 {
  emitterUpdateScript [
   &EmitterState()
   &SpawnRate(
    SpawnRate=&Lerp_Float(
     A=50
     B=150
     Alpha=&RemapRange(Input Value To Remap=$User.RightDot, Input Min=-1, Input Max=1, Output Min=0, Output Max=1)
    )
    Use Spawn Probability=true
   ) {enabled=false}
   &SpawnRate(
    SpawnGroup=1
    SpawnRate=&Lerp_Float(
     A=&FN_FrontEndFloatModifier(
      Gameplay=0
      Front End=&FloatFromCurve(FloatCurve=DataInterface() {propertyValues {Curve {keys [<at 0.5 0>,<at 2 1>]}}}, CurveIndex=$Engine.System.Age, Scale Curve=200)
     )
     B=200
     Alpha=$System.Timeline.NormalizedTime
    )
    Use Spawn Probability=true
   ) {module=SpawnRate001}
  ]
  particleSpawnScript [
   &InitializeParticle(
    Color=&LerpLinearColors(StartColor=<rgba 1 0.8756 0.4483 5>, EndColor=<rgba 1 1 1 5>, LerpFactor=$User.VariantIndex)
    Mesh Scale Max=<25 1 1>
    Mesh Scale Min=<10 1 1>
    Mesh Scale Mode=::NewEnumerator2
    UsePositionOffset=false
    Write Material Random=true
   )
   &InitialMeshOrientation(
    Mesh Orientation Mode=::NewEnumerator2
    Rotation=&RandomRangeVector(Minimum=<0 0 0>, Maximum=<1 0 0>)
    Use Rotation Vector=true
   )
   &TorusLocation(
    Align Torus With Axis=true
    Handle Radius=0
    Large Radius:NiagaraFloat=&RandomRangeFloat(Minimum=80, Maximum=150)
    MaskSpawnGroup=false
    Offset=<0 -180 0>
    Torus Axis=<0 1 0>
    Torus Distribution Mode:ENiagaraTorusDistributionMode=::NewEnumerator1
    Torus Origin:NiagaraPosition=&Add_Vector(
     A=Linked() {linkedVariable {name=Particles.Position, type=NiagaraPosition}}
     B=&MakeVector(X=0, Y=&MakeCustomFloatFromBool(Bool=$User.bIsFrontEndPreview, True Float=200, False Float=0), Z=0)
    )
    U Position:NiagaraFloat=&MakeCustomFloatFromBool(
     Bool=&SetBoolByFloatComparison(A=$User.RightDot, B=0, Comparison Type=::NewEnumerator4)
     True Float=&RandomRangeFloat(Minimum=0, Maximum=1)
     False Float=&Lerp_Float(
      A=&RandomRangeFloat(Minimum=0.25, Maximum=0.75)
      B=&RandomRangeFloat(Minimum=-0.25, Maximum=0.25)
      Alpha=&RemapRange(Input Value To Remap=$User.RightDot, Input Min=-1, Input Max=1, Output Min=0, Output Max=1)
     )
    )
    V Position:NiagaraFloat=&RandomRangeFloat(Minimum=0, Maximum=1)
   ) {enabled=false}
   &TorusLocation(
    Align Torus With Axis=true
    Handle Radius=0
    Large Radius:NiagaraFloat=&FN_FrontEndFloatModifier(
     Gameplay=&RandomRangeFloat(Minimum=180, Maximum=220)
     Front End=&RandomRangeFloat(Minimum=200, Maximum=250)
    )
    MaskSpawnGroup=false
    Module.Override Randomness=false
    Offset=<0 -180 0>
    Spawn Group Mask=1
    Torus Axis=<0 1 0>
    Torus Origin:NiagaraPosition=&Add_Vector(
     A=&FN_FrontEndVector3Modifier(Gameplay=<0 -100 50>, Front End=<0 -500 50>)
     B=&Lerp_Vector(A=<0 0 0>, B=<200 0 0>, Alpha=$User.RightDot)
    )
   ) {module=TorusLocation001}
   &AddVelocityFromPoint(
    Module.Use Falloff Distance=false
    Origin Offset=&Add_Vector(A=<0 -500 0>, B=&Lerp_Vector(A=<0 0 0>, B=<300 0 0>, Alpha=$User.RightDot))
    Velocity Strength=&FN_FrontEndFloatModifier(Gameplay=150, Front End=-150)
   )
   &AddVelocity(
    Module.Use Falloff Distance=false
    Origin Offset=&Add_Vector(A=<0 -500 0>, B=&Lerp_Vector(A=<0 0 0>, B=<500 0 0>, Alpha=$User.RightDot))
    Override Seed=false
    Velocity Mode=::NewEnumerator1
    Velocity Speed=200
   ) {module=AddVelocity001}
   &AddVelocity(Velocity=&Multiply_Vector(
     A=&RandomRangeVector(Minimum=<1 400 1>, Maximum=<1 600 1>)
     B=&MakeVector(X=1, Y=&MakeCustomFloatFromBool(Bool=$User.bIsFrontEndPreview, True Float=5, False Float=1), Z=1)
    ))
  ]
  particleUpdateScript [
   &ScaleVelocity(Velocity Scale=&VectorFromFloat(Value=&FloatFromCurve(
      FloatCurve=DataInterface() {propertyValues {Curve {keys [
          {interpMode=RCIM_Cubic, tangentMode=RCTM_User, time=0, value=0}
          {arriveTangent=2, interpMode=RCIM_Cubic, leaveTangent=2, time=0.1, value=0.975}
          {interpMode=RCIM_Cubic, tangentMode=RCTM_User, time=0.5, value=1}
          {arriveTangent=-2, interpMode=RCIM_Cubic, leaveTangent=-2, time=0.9, value=1}
          {tangentMode=RCTM_User, time=1, value=0}
         ]}}}
      CurveIndex=&RemapRange(Input Value To Remap=$User.RightDot, Input Min=-1, Input Max=1, Output Min=0, Output Max=1)
     )))
   &SolveForcesAndVelocity(Clamp Velocity=false, Limit Acceleration=false)
   &ParticleState()
   &ScaleMeshSize(Scale Factor=&VectorFromFloat(Value=&FloatFromCurve(FloatCurve=DataInterface() {propertyValues {Curve {keys [
          {tangentMode=RCTM_User, time=0, value=1}
          {arriveTangent=-0.0405, interpMode=RCIM_Cubic, leaveTangent=-0.0405, tangentMode=RCTM_User, time=0.7009, value=1}
          {arriveTangent=-8.0089, interpMode=RCIM_Cubic, leaveTangent=-8.0089, tangentMode=RCTM_User, time=1, value=0}
         ]}}})))
   &DynamicMaterialParameters(
    Index 0 Param 2=1
    Index 0 Param 3=1
    Index 0 Param 4=1
    Module.Param0WriteEnabled=true
    Module.Param1WriteEnabled=true
    Module.Param2WriteEnabled=true
    Module.Param3WriteEnabled=true
   )
   &ScaleColor(
    Scale Alpha=&FloatFromCurve(
     FloatCurve=DataInterface() {propertyValues {Curve {keys [
         {interpMode=RCIM_Cubic, tangentMode=RCTM_User, time=0, value=0}
         {interpMode=RCIM_Cubic, tangentMode=RCTM_User, time=0.5, value=1.5}
         {tangentMode=RCTM_User, time=1, value=0}
        ]}}}
     Scale Curve=&Lerp_Float(
      A=&FN_FrontEndFloatModifier(Gameplay=0.5, Front End=0.1)
      B=&Lerp_Float(A=0.1, B=0.5, Alpha=$System.Timeline001.Time)
      Alpha=$System.Timeline.Time
     )
    )
    ScaleA=true
    ScaleRGB=true
   )
  ]
  rendererClasses [&NiagaraMeshRendererProperties]
  renderers [{rendererClass=&NiagaraMeshRendererProperties, rendererIndex=0}]
 }
}
summary {
 systemName=NS_Glider_ForwardLake_Dire
 emitters [
  {emitterName=Smoke, rendererClasses [&NiagaraMeshRendererProperties]}
  {emitterName=Lighting, rendererClasses [&NiagaraSpriteRendererProperties, &NiagaraLightRendererProperties]}
  {emitterName=FlameRing, rendererClasses [&NiagaraMeshRendererProperties]}
  {emitterName=WindLines001, rendererClasses [&NiagaraMeshRendererProperties]}
 ]
 userVariables [
  User.bIsFrontEnd {type=NiagaraBool}
  User.bIsFrontEndPreview {type=NiagaraBool}
  User.bIsFullyDeployed {type=NiagaraBool}
  User.ForwardDot {type=NiagaraFloat, defaultValue=1, description=""}
  User.FuelAmount {type=NiagaraFloat, defaultValue=1, description=""}
  User.RightDot {type=NiagaraFloat, defaultValue=0, description=""}
  User.Variant {type=NiagaraFloat, defaultValue=0, description=""}
  User.VariantIndex {type=NiagaraFloat, defaultValue=0, description=""}
 ]
}
